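; PistolCanary_DirectorUI is a Blueprint that adds and removes the PistolCanary UI scene through the Dynamic UI Director.
; Two custom events drive it: PistolCanary_AddDynamicUI adds the scene asset and prints a "Hello" debug string, while PistolCanary_RemoveDynamicUI removes the same scene.
; BeginPlay, Tick and the construction script are empty, so nothing in this graph calls either event.
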(event EventBeginPlay)

(event EventTick
  (params (DeltaSeconds :type Float%20%28single-precision%29)))

(event Custom|PistolCanary_AddDynamicUI
  (DynamicUIDirector|AddScene "/PistolCanary/Layout/UI/PistolCanary_UI_Scene.PistolCanary_UI_Scene")
  (Development|PrintString "Hello"))

(event Custom|PistolCanary_RemoveDynamicUI
  (DynamicUIDirector|RemoveScene "/PistolCanary/Layout/UI/PistolCanary_UI_Scene.PistolCanary_UI_Scene"))

(fn ConstructionScript)
